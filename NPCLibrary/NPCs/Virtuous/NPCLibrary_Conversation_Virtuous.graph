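# Conversation database for the Virtuous NPC, who introduces itself as Graveheart, a protector who watches over the lost.
# Its entry point leads to one Speech node whose contextual lines cover a first-ever meeting, rotating flavor lines gated by per-match conversation count, and reactions to three specific player outfits.
# Priority curve rows rank the introduction and outfit reactions, with a default line as fallback.
# The speech then links out to the shared NPC library quests and basic services conversations.
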ConversationDatabase /NPCLibrary/NPCs/Virtuous/NPCLibrary_Conversation_Virtuous
entry_points [{tag=Conversation.Entry.Virtuous, destinations [#1]}]
nodes [
 ConversationEntryPointNode_7 {
  id=#1
  class=ConversationEntryPointNode
  outputs [#2]
  properties {entryTag {tagName=Conversation.Entry.Virtuous}}
 }
 Speech_C_4 {
  id=#2
  class=Speech_C
  outputs [#3, #4]
  properties {
   nodeName=""
   generalConfig {
    defaultMessage="Hush, child. There is beauty in the silence."
    contextualMessages [
     {
      message="I am Graveheart, protector of this place."
      contextRequirements [{requirement {class=FortControllerRequirement_IsFirstEverConversationWithNPC}}]
      priority {curve {rowName=Tandem.Speech.Priority.FirstEverInteraction}}
     }
     {
      message="Hush, child. There is beauty in the silence."
      contextRequirements [{requirement {class=FortControllerRequirement_CheckNPCMatchConversationCount, properties {nPCUniqueIdTag {tagName=AISpawnerData.Type.Tandem.Virtuous}}}}]
      priority {curve {rowName=None}}
     }
     {
      message="What do you see, when you gaze upon me? A monster?"
      contextRequirements [{requirement {class=FortControllerRequirement_CheckNPCMatchConversationCount, properties {nPCUniqueIdTag {tagName=AISpawnerData.Type.Tandem.Virtuous}}}}]
      priority {curve {rowName=None}}
     }
     {
      message="I watch over the lost, and sometimes, guide them back home."
      contextRequirements [{requirement {class=FortControllerRequirement_CheckNPCMatchConversationCount, properties {nPCUniqueIdTag {tagName=AISpawnerData.Type.Tandem.Virtuous}}}}]
      priority {curve {rowName=None}}
     }
     {
      message=I will always love the gentle strummings of the musicians.
      contextRequirements [{requirement {class=FortControllerRequirement_CheckNPCMatchConversationCount, properties {nPCUniqueIdTag {tagName=AISpawnerData.Type.Tandem.Virtuous}}}}]
      priority {curve {rowName=None}}
     }
     {
      message="I fear nothing, except perhaps snakes. They unnerve me."
      contextRequirements [{requirement {class=FortControllerRequirement_CheckNPCMatchConversationCount, properties {nPCUniqueIdTag {tagName=AISpawnerData.Type.Tandem.Virtuous}}}}]
      priority {curve {rowName=None}}
     }
     {
      message="I was known by another name, very long ago. That name is lost to the Styx."
      contextRequirements [{requirement {class=FortControllerRequirement_CheckNPCMatchConversationCount, properties {nPCUniqueIdTag {tagName=AISpawnerData.Type.Tandem.Virtuous}}}}]
      priority {curve {rowName=None}}
     }
     {
      message="Leave me be, deathly trickster."
      contextRequirements [{requirement {
         class=FortControllerRequirement_HasCID
         properties {allowedCIds [/BRCosmetics/Athena/Items/Cosmetics/Characters/Character_DarkStance_Inferno]}
        }}]
      priority {curve {rowName=Tandem.Speech.Priority.OutfitReaction}}
     }
     {
      message="Your blood runs cold as stone, but your heart burns hot."
      contextRequirements [{requirement {
         class=FortControllerRequirement_HasCID
         properties {allowedCIds [/BRCosmetics/Athena/Items/Cosmetics/Characters/Character_ReptilianOcean_Sleek]}
        }}]
      priority {curve {rowName=Tandem.Speech.Priority.OutfitReaction}}
     }
     {
      message="Once lost, now returned home. You are luckier than most."
      contextRequirements [{requirement {
         class=FortControllerRequirement_HasCID
         properties {allowedCIds [&CID_349_Athena_Commando_M_Banana]}
        }}]
      priority {curve {rowName=Tandem.Speech.Priority.OutfitReaction}}
     }
    ]
   }
  }
 }
 ConversationLinkNode_0 {id=#4, class=ConversationLinkNode, properties {remoteEntryTag {tagName=Conversation.Entry.BasicServices}}}
 ConversationLinkNode_1 {id=#3, class=ConversationLinkNode, properties {remoteEntryTag {tagName=Conversation.Entry.NPCLibrary.Quests}}}
]
